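# Sound cue for activating the Season 17 battle pass quest Teleportation Probe.
# A single wave player feeds the activation sound wave straight to the output, with no randomisation or mixing.
# It uses the shared Season 17 quest attenuation asset and plays under the Fort quests sound class.
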SoundCue /BattlepassS17/Sounds/S17Quests/TeleportationProbe/S17_BR_Teleportation_Probe_Activate_Cue
root=WavePlayer_0
nodes [{id=WavePlayer_0, properties {soundWaveAssetPtr=&/S17_BR_Teleportation_Probe_Activate}}]
properties {
 attenuationSettings=/BattlepassS17/Sounds/S17_Quest_Att
 maxDistance=4000
 soundClassObject=&Fort_Quests
 volumeMultiplier=1
}
