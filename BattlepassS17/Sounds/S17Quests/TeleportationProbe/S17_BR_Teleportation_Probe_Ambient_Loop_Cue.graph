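# Looping ambient sound cue for the Teleportation Probe in the Season 17 Battle Pass quest audio.
# A single looping Wave Player serves as the root and plays the probe's ambient loop wave, with no mixing or modulation nodes.
# It borrows the Season 7 quest attenuation asset for spatial falloff and routes through the Fort_Quests sound class.
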SoundCue /BattlepassS17/Sounds/S17Quests/TeleportationProbe/S17_BR_Teleportation_Probe_Ambient_Loop_Cue
root=WavePlayer_0
nodes [{
  id=WavePlayer_0
  properties {bLooping=true, soundWaveAssetPtr=&/S17_BR_Teleportation_Probe_Ambient_Loop}
 }]
properties {
 attenuationSettings=/Game/Sounds/Attenuation/Fort_Quest_Season7
 maxDistance=2400
 soundClassObject=&Fort_Quests
}
